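# Physics asset for the SprocketPoppy BoatCircus two-handed melee pickaxe cosmetic mesh.
# It holds a single capsule body on the Melee_TwoHanded_Handle bone, offset up the handle and rotated to align with it.
# There are no constraints, so the axe simulates and collides as one rigid shape.
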PhysicsAsset /BRCosmetics/Weapons/FORT_Melee/Pickaxe_SprocketPoppy_BoatCircus/Meshes/SprocketPoppy_BoatCircus_Axe_PhysicsAsset
bodies {Melee_TwoHanded_Handle [{shapeType=Capsule, length=34.9314, radius=24.4815, center=<0 0 17.9151>, rotation=<rot 0 -90 0>}]}
constraints []
